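# Physics asset for the crystal shards on the RareDelight backpack cosmetic, giving capsule bodies to the pack, master, heart and floating shard pivot bones.
# The master body joins to the pack and the heart joins to the master, both with free swing and twist.
# Every left and right shard pivot, both base and dyn, constrains directly to the master in a star layout, with modest swing limits and almost no twist.
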PhysicsAsset /BRCosmetics/Accessories/FORT_Backpacks/Backpack_F_MED_RareDelight/Meshes/F_MED_RareDelight_Pack_CristalShards_Physics
bodies {
 base_floatingShardPivot_04_l [{shapeType=Capsule, length=2, radius=1, shapeName=base_floatingShardPivot_04_l_capsule, center=<2 0 0>, rotation=<rot 90 0 0>}]
 base_floatingShardPivot_04_r [{shapeType=Capsule, length=2, radius=1, shapeName=base_floatingShardPivot_04_r_capsule, center=<2 0 0>, rotation=<rot 90 0 0>}]
 base_floatingShardPivot_05_l [{shapeType=Capsule, length=2, radius=1, shapeName=base_floatingShardPivot_05_l_capsule, center=<2 0 0>, rotation=<rot 90 0 0>}]
 base_floatingShardPivot_06_r [{shapeType=Capsule, length=2, radius=1, shapeName=base_floatingShardPivot_06_r_capsule, center=<2 0 0>, rotation=<rot 90 0 0>}]
 base_floatingShardPivot_07_r [{shapeType=Capsule, length=2, radius=1, shapeName=base_floatingShardPivot_07_r_capsule, center=<2 0 0>, rotation=<rot 90 0 0>}]
 base_heart [{
   shapeType=Capsule
   length=7.7808
   radius=5.5921
   shapeName=base_heart_capsule
   center=<-0.3298 0.1888 0.2166>
   rotation=<rot 0 91.0091 -85.1559>
  }]
 base_master [{
   shapeType=Capsule
   length=3.8214
   radius=3.6217
   shapeName=base_master_capsule
   center=<0.1179 -0.1959 -0.0068>
   rotation=<rot 8.2985 -90 -68.6738>
  }]
 dyn_floatingShardPivot_01_l [{shapeType=Capsule, length=2, radius=1, shapeName=dyn_floatingShardPivot_01_l_capsule, center=<2 0 0>, rotation=<rot 90 0 0>}]
 dyn_floatingShardPivot_01_r [{shapeType=Capsule, length=2, radius=1, shapeName=dyn_floatingShardPivot_01_r_capsule, center=<2 0 0>, rotation=<rot 90 0 0>}]
 dyn_floatingShardPivot_02_l [{shapeType=Capsule, length=2, radius=1, shapeName=dyn_floatingShardPivot_02_l_capsule, center=<2 0 0>, rotation=<rot 90 0 0>}]
 dyn_floatingShardPivot_02_r [{shapeType=Capsule, length=2, radius=1, shapeName=dyn_floatingShardPivot_02_r_capsule, center=<2 0 0>, rotation=<rot 90 0 0>}]
 dyn_floatingShardPivot_03_l [{shapeType=Capsule, length=2, radius=1, shapeName=dyn_floatingShardPivot_03_l_capsule, center=<2 0 0>, rotation=<rot 90 0 0>}]
 dyn_floatingShardPivot_03_r [{shapeType=Capsule, length=2, radius=1, shapeName=dyn_floatingShardPivot_03_r_capsule, center=<2 0 0>, rotation=<rot 90 0 0>}]
 dyn_floatingShardPivot_06_l [{shapeType=Capsule, length=2, radius=1, shapeName=dyn_floatingShardPivot_06_l_capsule, center=<2 0 0>, rotation=<rot 90 0 0>}]
 pack [{shapeType=Capsule, length=0.505, radius=0.505, shapeName=pack_capsule, rotation=<rot 0 -90 0>}]
}
constraints [
 {
  bone1Name=base_master
  bone2Name=pack
  swing1LimitDegrees=45
  swing1Motion=Free
  swing2LimitDegrees=45
  swing2Motion=Free
  twistLimitDegrees=45
  twistMotion=Free
 }
 {
  bone1Name=base_heart
  bone2Name=base_master
  swing1LimitDegrees=45
  swing1Motion=Free
  swing2LimitDegrees=45
  swing2Motion=Free
  twistLimitDegrees=45
  twistMotion=Free
 }
 {bone1Name=base_floatingShardPivot_04_l, bone2Name=base_master, swing1LimitDegrees=25, swing2LimitDegrees=25, twistLimitDegrees=1}
 {bone1Name=base_floatingShardPivot_05_l, bone2Name=base_master, swing1LimitDegrees=25, swing2LimitDegrees=25, twistLimitDegrees=1}
 {bone1Name=dyn_floatingShardPivot_06_l, bone2Name=base_master, swing1LimitDegrees=15, swing2LimitDegrees=15, twistLimitDegrees=1}
 {bone1Name=base_floatingShardPivot_04_r, bone2Name=base_master, swing1LimitDegrees=15, swing2LimitDegrees=15, twistLimitDegrees=1}
 {bone1Name=base_floatingShardPivot_06_r, bone2Name=base_master, swing1LimitDegrees=25, swing2LimitDegrees=25, twistLimitDegrees=1}
 {bone1Name=base_floatingShardPivot_07_r, bone2Name=base_master, swing1LimitDegrees=25, swing2LimitDegrees=25, twistLimitDegrees=1}
 {bone1Name=dyn_floatingShardPivot_03_r, bone2Name=base_master, swing1LimitDegrees=30, swing2LimitDegrees=30, twistLimitDegrees=1}
 {bone1Name=dyn_floatingShardPivot_01_l, bone2Name=base_master, swing1LimitDegrees=25, swing2LimitDegrees=25, twistLimitDegrees=1}
 {bone1Name=dyn_floatingShardPivot_02_l, bone2Name=base_master, swing1LimitDegrees=25, swing2LimitDegrees=25, twistLimitDegrees=1}
 {bone1Name=dyn_floatingShardPivot_03_l, bone2Name=base_master, swing1LimitDegrees=25, swing2LimitDegrees=25, twistLimitDegrees=1}
 {bone1Name=dyn_floatingShardPivot_01_r, bone2Name=base_master, swing1LimitDegrees=25, swing2LimitDegrees=25, twistLimitDegrees=1}
 {bone1Name=dyn_floatingShardPivot_02_r, bone2Name=base_master, swing1LimitDegrees=25, swing2LimitDegrees=25, twistLimitDegrees=1}
]
