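# Physics asset for the RareDelight backpack cosmetic, defining bodies and joints for its heart, ribbons and ribbon strings.
# Every bone gets the same small capsule; the heart and ribbon base bodies link to base_master with free swing and twist.
# Mirrored dyn_ chains hang from those bases: three-segment strings with narrow angular limits, and two-segment ribbons whose first joint locks rotation and second allows slight bend.
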PhysicsAsset /BRCosmetics/Accessories/FORT_Backpacks/Backpack_F_MED_RareDelight/Meshes/F_MED_RareDelight_Pack_Physics
bodies {
 base_heart [{shapeType=Capsule, length=2, radius=1, shapeName=base_heart_capsule, center=<2 0 0>, rotation=<rot 90 0 0>}]
 base_master [{shapeType=Capsule, length=2, radius=1, shapeName=base_master_capsule, center=<2 0 0>, rotation=<rot 90 0 0>}]
 base_ribbonString_00_l [{shapeType=Capsule, length=2, radius=1, shapeName=base_ribbonString_00_l_capsule, center=<2 0 0>, rotation=<rot 90 0 0>}]
 base_ribbonString_00_r [{shapeType=Capsule, length=2, radius=1, shapeName=base_ribbonString_00_r_capsule, center=<2 0 0>, rotation=<rot 90 0 0>}]
 base_ribbon_00_l [{shapeType=Capsule, length=2, radius=1, shapeName=base_ribbon_00_l_capsule, center=<2 0 0>, rotation=<rot 90 0 0>}]
 base_ribbon_00_r [{shapeType=Capsule, length=2, radius=1, shapeName=base_ribbon_00_r_capsule, center=<2 0 0>, rotation=<rot 90 0 0>}]
 dyn_ribbonString_00_l [{shapeType=Capsule, length=2, radius=1, shapeName=dyn_ribbonString_00_l_capsule, center=<2 0 0>, rotation=<rot 90 0 0>}]
 dyn_ribbonString_00_r [{shapeType=Capsule, length=2, radius=1, shapeName=dyn_ribbonString_00_r_capsule, center=<2 0 0>, rotation=<rot 90 0 0>}]
 dyn_ribbonString_01_l [{shapeType=Capsule, length=2, radius=1, shapeName=dyn_ribbonString_01_l_capsule, center=<2 0 0>, rotation=<rot 90 0 0>}]
 dyn_ribbonString_01_r [{shapeType=Capsule, length=2, radius=1, shapeName=dyn_ribbonString_01_r_capsule, center=<2 0 0>, rotation=<rot 90 0 0>}]
 dyn_ribbonString_02_l [{shapeType=Capsule, length=2, radius=1, shapeName=dyn_ribbonString_02_l_capsule, center=<2 0 0>, rotation=<rot 90 0 0>}]
 dyn_ribbonString_02_r [{shapeType=Capsule, length=2, radius=1, shapeName=dyn_ribbonString_02_r_capsule, center=<2 0 0>, rotation=<rot 90 0 0>}]
 dyn_ribbon_00_l [{shapeType=Capsule, length=2, radius=1, shapeName=dyn_ribbon_00_l_capsule, center=<2 0 0>, rotation=<rot 90 0 0>}]
 dyn_ribbon_00_r [{shapeType=Capsule, length=2, radius=1, shapeName=dyn_ribbon_00_r_capsule, center=<2 0 0>, rotation=<rot 90 0 0>}]
 dyn_ribbon_01_l [{shapeType=Capsule, length=2, radius=1, shapeName=dyn_ribbon_01_l_capsule, center=<2 0 0>, rotation=<rot 90 0 0>}]
 dyn_ribbon_01_r [{shapeType=Capsule, length=2, radius=1, shapeName=dyn_ribbon_01_r_capsule, center=<2 0 0>, rotation=<rot 90 0 0>}]
}
constraints [
 {bone1Name=dyn_ribbonString_00_l, bone2Name=base_ribbonString_00_l, swing1LimitDegrees=10, swing2LimitDegrees=25, twistLimitDegrees=1}
 {bone1Name=dyn_ribbonString_01_l, bone2Name=dyn_ribbonString_00_l, swing1LimitDegrees=10, swing2LimitDegrees=25, twistLimitDegrees=1}
 {bone1Name=dyn_ribbonString_02_l, bone2Name=dyn_ribbonString_01_l, swing1LimitDegrees=10, swing2LimitDegrees=25, twistLimitDegrees=1}
 {bone1Name=dyn_ribbonString_00_r, bone2Name=base_ribbonString_00_r, swing1LimitDegrees=10, swing2LimitDegrees=27, twistLimitDegrees=1}
 {bone1Name=dyn_ribbonString_01_r, bone2Name=dyn_ribbonString_00_r, swing1LimitDegrees=10, swing2LimitDegrees=27, twistLimitDegrees=1}
 {bone1Name=dyn_ribbonString_02_r, bone2Name=dyn_ribbonString_01_r, swing1LimitDegrees=10, swing2LimitDegrees=27, twistLimitDegrees=1}
 {
  bone1Name=dyn_ribbon_00_l
  bone2Name=base_ribbon_00_l
  swing1LimitDegrees=1
  swing1Motion=Locked
  swing2LimitDegrees=1
  swing2Motion=Locked
  twistLimitDegrees=1
  twistMotion=Locked
 }
 {bone1Name=dyn_ribbon_01_l, bone2Name=dyn_ribbon_00_l, swing1LimitDegrees=2, swing2LimitDegrees=12, twistLimitDegrees=1}
 {
  bone1Name=dyn_ribbon_00_r
  bone2Name=base_ribbon_00_r
  swing1LimitDegrees=1
  swing1Motion=Locked
  swing2LimitDegrees=1
  swing2Motion=Locked
  twistLimitDegrees=1
  twistMotion=Locked
 }
 {bone1Name=dyn_ribbon_01_r, bone2Name=dyn_ribbon_00_r, swing1LimitDegrees=2, swing2LimitDegrees=12, twistLimitDegrees=1}
 {
  bone1Name=base_heart
  bone2Name=base_master
  swing1LimitDegrees=45
  swing1Motion=Free
  swing2LimitDegrees=45
  swing2Motion=Free
  twistLimitDegrees=45
  twistMotion=Free
 }
 {
  bone1Name=base_ribbon_00_r
  bone2Name=base_master
  swing1LimitDegrees=45
  swing1Motion=Free
  swing2LimitDegrees=45
  swing2Motion=Free
  twistLimitDegrees=45
  twistMotion=Free
 }
 {
  bone1Name=base_ribbon_00_l
  bone2Name=base_master
  swing1LimitDegrees=45
  swing1Motion=Free
  swing2LimitDegrees=45
  swing2Motion=Free
  twistLimitDegrees=45
  twistMotion=Free
 }
 {
  bone1Name=base_ribbonString_00_r
  bone2Name=base_master
  swing1LimitDegrees=45
  swing1Motion=Free
  swing2LimitDegrees=45
  swing2Motion=Free
  twistLimitDegrees=45
  twistMotion=Free
 }
 {
  bone1Name=base_ribbonString_00_l
  bone2Name=base_master
  swing1LimitDegrees=45
  swing1Motion=Free
  swing2LimitDegrees=45
  swing2Motion=Free
  twistLimitDegrees=45
  twistMotion=Free
 }
]
