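# Anim Blueprint for the RareDelight backpack cosmetic, adding physics and floating motion to its crystal shards.
# The input pose gets a pack offset, then two RigidBody passes (shards, pack), whose alpha drops while skydiving.
# Cached poses then feed additive bone translations driven by five phase-offset MakeWaveVector values, Float A–E.
# The per-shard sine vectors and GravityOverride update each frame but drive nothing, and a third RigidBody node is unconnected.
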AnimBlueprint /BRCosmetics/Accessories/FORT_Backpacks/Backpack_F_MED_RareDelight/Meshes/F_MED_RareDelight_Pack_AnimBP
target_skeleton=&/F_MED_RareDelight_Pack_Skeleton
graphs [
 AnimGraph {nodes [
   RigidBody!(ComponentPose, Alpha, OverrideWorldGravity:Vector="0,0,0", ExternalForce) -> (Pose) {
    bClampLinearTranslationLimitToRefPose=true
    bOverrideWorldGravity=true
    overridePhysicsAsset=&/F_MED_RareDelight_Pack_CristalShards_Physics
    simulationSpace=BaseBoneSpace
    alphaBoolBlend {blendOption=Linear}
    alphaScaleBiasClamp {bInterpResult=true}
    baseBoneRef {boneName=base_metal_chips}
    componentAppliedLinearAccClamp=<3000 3000 3000>
    componentLinearAccScale=<0.4 0.3 0.3>
    componentLinearVelScale=<0.4 0.3 0.3>
    overrideWorldGravity {z=-980}
    simSpaceSettings {velocityScaleZ=0.2, worldAlpha=0.1, externalLinearDragV=<0.2 0.2>}
   }
   n1 = "Transform (Modify) Bone - Bone: dyn_floatingShard_04_l"(
    ComponentPose=Local To Component(LocalPose=Use cached pose 'Translate Top Shards '())
    Alpha=1
    Translation=Get Float D() -> (Float D:Vector)
    Rotation="0,0,0"
    Scale="1,1,1"
   ) {
    translationMode=BMM_Additive
    translationSpace=BCS_BoneSpace
    alphaBoolBlend {blendOption=Linear}
    boneToModify {boneName=dyn_floatingShard_04_l}
   }
   n2 = "Transform (Modify) Bone - Bone: dyn_floatingShard_05_l"(ComponentPose=@n1, Alpha=1, Translation=Get Float C() -> (Float C:Vector), Rotation="0,0,0", Scale="1,1,1") {
    translationMode=BMM_Additive
    translationSpace=BCS_BoneSpace
    alphaBoolBlend {blendOption=Linear}
    boneToModify {boneName=dyn_floatingShard_05_l}
   }
   n3 = "Transform (Modify) Bone - Bone: dyn_floatingShard_06_l"(ComponentPose=@n2, Alpha=1, Translation=Get Float B() -> (Float B:Vector), Rotation="0,0,0", Scale="1,1,1") {
    translationMode=BMM_Additive
    translationSpace=BCS_BoneSpace
    alphaBoolBlend {blendOption=Linear}
    boneToModify {boneName=dyn_floatingShard_06_l}
   }
   n4 = "Transform (Modify) Bone - Bone: dyn_floatingShard_03_r"(ComponentPose=@n3, Alpha=1, Translation=Get Float D(), Rotation="0,0,0", Scale="1,1,1") {
    translationMode=BMM_Additive
    translationSpace=BCS_BoneSpace
    alphaBoolBlend {blendOption=Linear}
    boneToModify {boneName=dyn_floatingShard_03_r}
   }
   n5 = "Transform (Modify) Bone - Bone: dyn_floatingShard_04_r"(ComponentPose=@n4, Alpha=1, Translation=Get Float B(), Rotation="0,0,0", Scale="1,1,1") {
    translationMode=BMM_Additive
    translationSpace=BCS_BoneSpace
    alphaBoolBlend {blendOption=Linear}
    boneToModify {boneName=dyn_floatingShard_04_r}
   }
   n6 = "Transform (Modify) Bone - Bone: dyn_floatingShard_05_r"(ComponentPose=@n5, Alpha=1, Translation=Get Float C(), Rotation="0,0,0", Scale="1,1,1") {
    translationMode=BMM_Additive
    translationSpace=BCS_BoneSpace
    alphaBoolBlend {blendOption=Linear}
    boneToModify {boneName=dyn_floatingShard_05_r}
   }
   n7 = "Transform (Modify) Bone - Bone: dyn_floatingShard_06_r"(ComponentPose=@n6, Alpha=1, Translation=Get Float A() -> (Float A:Vector), Rotation="0,0,0", Scale="1,1,1") {
    translationMode=BMM_Additive
    translationSpace=BCS_BoneSpace
    alphaBoolBlend {blendOption=Linear}
    boneToModify {boneName=dyn_floatingShard_06_r}
   }
   n8 = "Transform (Modify) Bone - Bone: dyn_floatingShard_07_r"(ComponentPose=@n7, Alpha=1, Translation=Get Float E() -> (Float E:Vector), Rotation="0,0,0", Scale="1,1,1") {
    translationMode=BMM_Additive
    translationSpace=BCS_BoneSpace
    alphaBoolBlend {blendOption=Linear}
    boneToModify {boneName=dyn_floatingShard_07_r}
   }
   Output Pose(Result=Component To Local(ComponentPose=@n8)) {layerGroup=None}
   n9 = "Transform (Modify) Bone - Bone: pack"(ComponentPose=Local To Component(LocalPose=Input Pose()), Translation="0,1.7,0", Rotation="0,0,0") {
    translationMode=BMM_Additive
    translationSpace=BCS_BoneSpace
    alphaBoolBlend {blendOption=Linear}
    boneToModify {boneName=pack}
    translation {y=1.7}
   }
   n10 = RigidBody(ComponentPose=@n9, Alpha=Get Skydive Alpha() -> (Skydive Alpha:double), OverrideWorldGravity="0,0,0") {
    bOverrideWorldGravity=true
    overridePhysicsAsset=&F_MED_RareDelight_Pack_CristalShards_Physics
    simulationSpace=BaseBoneSpace
    alphaBoolBlend {blendOption=Linear}
    alphaScaleBiasClamp {bInterpResult=true}
    baseBoneRef {boneName=root}
    componentAppliedLinearAccClamp=<3000 3000 3000>
    componentLinearAccScale=<0.4 0.3 0.3>
    componentLinearVelScale=<0.4 0.3 0.3>
    simSpaceSettings {
     maxAngularAcceleration=1000
     maxAngularVelocity=3
     maxLinearAcceleration=700
     maxLinearVelocity=500
     velocityScaleZ=0
     worldAlpha=0.2
     externalLinearDragV=<0.2 0.2>
    }
   }
   n11 = RigidBody(ComponentPose=@n10, Alpha=Get Skydive Alpha(), OverrideWorldGravity="0,0,-500") {
    bClampLinearTranslationLimitToRefPose=true
    bOverrideWorldGravity=true
    overridePhysicsAsset=&/F_MED_RareDelight_Pack_Physics
    simulationSpace=BaseBoneSpace
    alphaBoolBlend {blendOption=Linear}
    alphaScaleBiasClamp {bInterpResult=true}
    baseBoneRef {boneName=root}
    componentAppliedLinearAccClamp=<1000 1500 3000>
    componentLinearAccScale=<0.1 0.15 0.2>
    componentLinearVelScale=<0.1 0.15 0.2>
    simSpaceSettings {
     maxAngularAcceleration=1000
     maxAngularVelocity=3
     maxLinearAcceleration=700
     maxLinearVelocity=500
     velocityScaleZ=0
     worldAlpha=0.2
     externalLinearDragV=<0.2 0.2>
    }
   }
   Save cached pose 'Rigid Body'(Pose=Component To Local(ComponentPose=@n11)) {cacheName=Rigid Body}
   n12 = "Transform (Modify) Bone - Bone: dyn_floatingShard_01_l"(
    ComponentPose=Local To Component(LocalPose=Use cached pose 'Rigid Body'())
    Alpha=1
    Translation=Get Float A()
    Rotation="0,0,0"
    Scale="1,1,1"
   ) {
    translationMode=BMM_Additive
    translationSpace=BCS_BoneSpace
    alphaBoolBlend {blendOption=Linear}
    boneToModify {boneName=dyn_floatingShard_01_l}
   }
   n13 = "Transform (Modify) Bone - Bone: dyn_floatingShard_02_l"(ComponentPose=@n12, Alpha=1, Translation=Get Float B(), Rotation="0,0,0", Scale="1,1,1") {
    translationMode=BMM_Additive
    translationSpace=BCS_BoneSpace
    alphaBoolBlend {blendOption=Linear}
    boneToModify {boneName=dyn_floatingShard_02_l}
   }
   n14 = "Transform (Modify) Bone - Bone: dyn_floatingShard_03_l"(ComponentPose=@n13, Alpha=1, Translation=Get Float A(), Rotation="0,0,0", Scale="1,1,1") {
    translationMode=BMM_Additive
    translationSpace=BCS_BoneSpace
    alphaBoolBlend {blendOption=Linear}
    boneToModify {boneName=dyn_floatingShard_03_l}
   }
   n15 = "Transform (Modify) Bone - Bone: dyn_floatingShard_01_r"(ComponentPose=@n14, Alpha=1, Translation=Get Float B(), Rotation="0,0,0", Scale="1,1,1") {
    translationMode=BMM_Additive
    translationSpace=BCS_BoneSpace
    alphaBoolBlend {blendOption=Linear}
    boneToModify {boneName=dyn_floatingShard_01_r}
   }
   n16 = "Transform (Modify) Bone - Bone: dyn_floatingShard_02_r"(ComponentPose=@n15, Alpha=1, Translation=Get Float A(), Rotation="0,0,0", Scale="1,1,1") {
    translationMode=BMM_Additive
    translationSpace=BCS_BoneSpace
    alphaBoolBlend {blendOption=Linear}
    boneToModify {boneName=dyn_floatingShard_02_r}
   }
   Save cached pose 'Translate Top Shards '(Pose=Component To Local(ComponentPose=@n16)) {cacheName="Translate Top Shards "}
   Get Translate Shard 01 l() -> (Translate Shard 01 l:Vector)
   Get Translate Shard 04 r() -> (Translate Shard 04 r:Vector)
   Get Translate Shard 05 r() -> (Translate Shard 05 r:Vector)
   Get Translate Shard 06 r() -> (Translate Shard 06 r:Vector)
   Get Translate Shard 07 r() -> (Translate Shard 07 r:Vector)
   Get Float A()
   Get Translate Shard 02 l() -> (Translate Shard 02 l:Vector)
   Get Float C()
   Get Float C()
   Get Float B()
   Get Float A()
   Get Float B()
   Get GravityOverride()
   Get Translate Shard 03 l() -> (Translate Shard 03 l:Vector)
   Get Float E()
   Get Translate Shard 01 r() -> (Translate Shard 01 r:Vector)
   Get Translate Shard 02 r() -> (Translate Shard 02 r:Vector)
   Get Translate Shard 04 l() -> (Translate Shard 04 l:Vector)
   Get Translate Shard 05 l() -> (Translate Shard 05 l:Vector)
   Get Translate Shard 06 l() -> (Translate Shard 06 l:Vector)
   Get Translate Shard 03 r() -> (Translate Shard 03 r:Vector)
  ]}
 EventGraph {nodes [
   Event BlueprintUpdateAnimation()
   > n1 = Sequence!(execute) -> (
    then_0
    then_1
    then_2:Exec
    then_3:Exec
    then_4:Exec
    then_5:Exec
    then_6:Exec
    then_7:Exec
    then_8:Exec
    then_9:Exec
    then_10:Exec
    then_11:Exec
    then_12:Exec
    then_13:Exec
    then_14:Exec
   )
   > n2 = Gravity Override State Function DEPRECATED(CharacterPartAnimInstance=Self-Reference(), GameTimeInSeconds=GetGameTimeInSeconds(), GravityOverrideParams=Get Gravity Override Params())
   > Set GravityOverride(GravityOverride=@n2)
   > Set Skydive Alpha(Skydive Alpha:double=SelectFloat(A=0.1, B=0.5, bPickA=Get bIsSkyDiving())) -> (Output_Get:double)
   n3 = GetRealTimeSeconds()
   n4 = "Sin (Radians)"(A=float * float(A=float * float(A=1.5, B=0.8) {is=title K2Node_PromotableOperator}, B=@n3))
   n5 = "Sin (Radians)"(A=float * float(A=float * float(A=0.4, B=-0.8) {is=title K2Node_PromotableOperator}, B=@n3) {is=title K2Node_CommutativeAssociativeBinaryOperator})
   n6 = "Sin (Radians)"(A=float * float(A=float * float(A=1.6, B=1.25) {is=title K2Node_PromotableOperator}, B=@n3) {is=title K2Node_CommutativeAssociativeBinaryOperator})
   Set Translate Shard 02 r(
    execute=@n1.then_5
    Translate Shard 02 r:Vector=MakeVector(X=float / float(A=@n4, B=-3.5), Y=float / float(A=@n6, B=0.95), Z=float / float(A=@n5, B=-2))
   ) -> (Output_Get:Vector)
   n7 = "Sin (Radians)"(A=float * float(A=float * float(A=1.5, B=0.8) {is=title K2Node_PromotableOperator}, B=@n3) {is=title K2Node_CommutativeAssociativeBinaryOperator})
   n8 = "Sin (Radians)"(A=float * float(A=float * float(A=0.4, B=-0.8) {is=title K2Node_PromotableOperator}, B=@n3) {is=title K2Node_CommutativeAssociativeBinaryOperator})
   n9 = "Sin (Radians)"(A=float * float(A=float * float(A=1.5, B=1.2) {is=title K2Node_PromotableOperator}, B=@n3) {is=title K2Node_CommutativeAssociativeBinaryOperator})
   Set Translate Shard 01 r(
    execute=@n1.then_4
    Translate Shard 01 r:Vector=MakeVector(X=float / float(A=@n7, B=-3.5), Y=float / float(A=@n9, B=1), Z=float / float(A=@n8, B=-2))
   ) -> (Output_Get:Vector)
   n10 = "Sin (Radians)"(A=float * float(A=float * float(A=1.5, B=0.8) {is=title K2Node_PromotableOperator}, B=@n3) {is=title K2Node_CommutativeAssociativeBinaryOperator})
   n11 = "Sin (Radians)"(A=float * float(A=float * float(A=0.4, B=-1) {is=title K2Node_PromotableOperator}, B=@n3) {is=title K2Node_CommutativeAssociativeBinaryOperator})
   n12 = "Sin (Radians)"(A=float * float(A=float * float(A=1.5, B=1.25) {is=title K2Node_PromotableOperator}, B=@n3) {is=title K2Node_CommutativeAssociativeBinaryOperator})
   Set Translate Shard 03 l(
    execute=@n1.then_3
    Translate Shard 03 l:Vector=MakeVector(X=float / float(A=@n10, B=3.5), Y=float / float(A=@n12, B=1), Z=float / float(A=@n11, B=-2))
   ) -> (Output_Get:Vector)
   n13 = "Sin (Radians)"(A=float * float(A=float * float(A=1.5, B=0.8) {is=title K2Node_PromotableOperator}, B=@n3) {is=title K2Node_CommutativeAssociativeBinaryOperator})
   n14 = "Sin (Radians)"(A=float * float(A=float * float(A=0.4, B=-0.8) {is=title K2Node_PromotableOperator}, B=@n3) {is=title K2Node_CommutativeAssociativeBinaryOperator})
   n15 = "Sin (Radians)"(A=float * float(A=float * float(A=1.5, B=1.2) {is=title K2Node_PromotableOperator}, B=@n3) {is=title K2Node_CommutativeAssociativeBinaryOperator})
   Set Translate Shard 02 l(
    execute=@n1.then_2
    Translate Shard 02 l:Vector=MakeVector(X=float / float(A=@n13, B=-3.5), Y=float / float(A=@n15, B=1), Z=float / float(A=@n14, B=-2.4))
   ) -> (Output_Get:Vector)
   n16 = "Sin (Radians)"(A=float * float(A=float * float(A=1.5, B=0.8) {is=title K2Node_PromotableOperator}, B=@n3) {is=title K2Node_CommutativeAssociativeBinaryOperator})
   n17 = "Sin (Radians)"(A=float * float(A=float * float(A=0.4, B=-0.8) {is=title K2Node_PromotableOperator}, B=@n3) {is=title K2Node_CommutativeAssociativeBinaryOperator})
   n18 = "Sin (Radians)"(A=float * float(A=float * float(A=1.5, B=1.25) {is=title K2Node_PromotableOperator}, B=@n3) {is=title K2Node_CommutativeAssociativeBinaryOperator})
   Set Translate Shard 01 l(
    execute=@n1.then_1
    Translate Shard 01 l:Vector=MakeVector(X=float / float(A=@n16, B=3.5), Y=float / float(A=@n18, B=0.95), Z=float / float(A=@n17, B=-2.4))
   ) -> (Output_Get:Vector)
   n19 = GetRealTimeSeconds()
   n20 = "Sin (Radians)"(A=float * float(A=float * float(A=1, B=0.8) {is=title K2Node_PromotableOperator}, B=@n19) {is=title K2Node_CommutativeAssociativeBinaryOperator})
   n21 = "Sin (Radians)"(A=float * float(A=float * float(A=0.4, B=1) {is=title K2Node_PromotableOperator}, B=@n19) {is=title K2Node_CommutativeAssociativeBinaryOperator})
   n22 = "Sin (Radians)"(A=float * float(A=float * float(A=1.5, B=1.2) {is=title K2Node_PromotableOperator}, B=@n19) {is=title K2Node_CommutativeAssociativeBinaryOperator})
   Set Translate Shard 04 l(
    execute=@n1.then_6
    Translate Shard 04 l:Vector=MakeVector(X=float / float(A=@n20, B=2.8), Y=float / float(A=@n22, B=1), Z=float / float(A=@n21, B=-2))
   ) -> (Output_Get:Vector)
   n23 = "Sin (Radians)"(A=float * float(A=float * float(A=1.3, B=0.8) {is=title K2Node_PromotableOperator}, B=@n19) {is=title K2Node_CommutativeAssociativeBinaryOperator})
   n24 = "Sin (Radians)"(A=float * float(A=float * float(A=0.4, B=-0.8) {is=title K2Node_PromotableOperator}, B=@n19) {is=title K2Node_CommutativeAssociativeBinaryOperator})
   n25 = "Sin (Radians)"(A=float * float(A=float * float(A=1.5, B=1.25) {is=title K2Node_PromotableOperator}, B=@n19) {is=title K2Node_CommutativeAssociativeBinaryOperator})
   Set Translate Shard 05 l(
    execute=@n1.then_7
    Translate Shard 05 l:Vector=MakeVector(X=float / float(A=@n23, B=-3), Y=float / float(A=@n25, B=1), Z=float / float(A=@n24, B=-2))
   ) -> (Output_Get:Vector)
   n26 = "Sin (Radians)"(A=float * float(A=float * float(A=1, B=0.8) {is=title K2Node_PromotableOperator}, B=@n19) {is=title K2Node_CommutativeAssociativeBinaryOperator})
   n27 = "Sin (Radians)"(A=float * float(A=float * float(A=0.4, B=-0.8) {is=title K2Node_PromotableOperator}, B=@n19) {is=title K2Node_CommutativeAssociativeBinaryOperator})
   n28 = "Sin (Radians)"(A=float * float(A=float * float(A=1.5, B=1.2) {is=title K2Node_PromotableOperator}, B=@n19) {is=title K2Node_CommutativeAssociativeBinaryOperator})
   Set Translate Shard 06 l(
    execute=@n1.then_8
    Translate Shard 06 l:Vector=MakeVector(X=float / float(A=@n26, B=3), Y=float / float(A=@n28, B=1), Z=float / float(A=@n27, B=-2))
   ) -> (Output_Get:Vector)
   n29 = "Sin (Radians)"(A=float * float(A=float * float(A=1, B=0.8) {is=title K2Node_PromotableOperator}, B=@n19) {is=title K2Node_CommutativeAssociativeBinaryOperator})
   n30 = "Sin (Radians)"(A=float * float(A=float * float(A=0.4, B=-0.8) {is=title K2Node_PromotableOperator}, B=@n19) {is=title K2Node_CommutativeAssociativeBinaryOperator})
   n31 = "Sin (Radians)"(A=float * float(A=float * float(A=1.5, B=1.25) {is=title K2Node_PromotableOperator}, B=@n19) {is=title K2Node_CommutativeAssociativeBinaryOperator})
   Set Translate Shard 03 r(
    execute=@n1.then_9
    Translate Shard 03 r:Vector=MakeVector(X=float / float(A=@n29, B=-2.8), Y=float / float(A=@n31, B=0.95), Z=float / float(A=@n30, B=-2))
   ) -> (Output_Get:Vector)
   n32 = "Sin (Radians)"(A=float * float(A=float * float(A=1, B=0.8) {is=title K2Node_PromotableOperator}, B=@n19) {is=title K2Node_CommutativeAssociativeBinaryOperator})
   n33 = "Sin (Radians)"(A=float * float(A=float * float(A=0.4, B=-0.8) {is=title K2Node_PromotableOperator}, B=@n19) {is=title K2Node_CommutativeAssociativeBinaryOperator})
   n34 = "Sin (Radians)"(A=float * float(A=float * float(A=1.5, B=1.2) {is=title K2Node_PromotableOperator}, B=@n19) {is=title K2Node_CommutativeAssociativeBinaryOperator})
   Set Translate Shard 04 r(
    execute=@n1.then_10
    Translate Shard 04 r:Vector=MakeVector(X=float / float(A=@n32, B=-2.5), Y=float / float(A=@n34, B=1), Z=float / float(A=@n33, B=-2))
   ) -> (Output_Get:Vector)
   n35 = "Sin (Radians)"(A=float * float(A=float * float(A=1, B=0.8) {is=title K2Node_PromotableOperator}, B=@n19) {is=title K2Node_CommutativeAssociativeBinaryOperator})
   n36 = "Sin (Radians)"(A=float * float(A=float * float(A=0.4, B=-0.8) {is=title K2Node_PromotableOperator}, B=@n19) {is=title K2Node_CommutativeAssociativeBinaryOperator})
   n37 = "Sin (Radians)"(A=float * float(A=float * float(A=1.5, B=1.25) {is=title K2Node_PromotableOperator}, B=@n19) {is=title K2Node_CommutativeAssociativeBinaryOperator})
   Set Translate Shard 05 r(
    execute=@n1.then_11
    Translate Shard 05 r:Vector=MakeVector(X=float / float(A=@n35, B=-2.8), Y=float / float(A=@n37, B=1), Z=float / float(A=@n36, B=-2))
   ) -> (Output_Get:Vector)
   n38 = "Sin (Radians)"(A=float * float(A=float * float(A=1, B=0.7) {is=title K2Node_PromotableOperator}, B=@n19) {is=title K2Node_CommutativeAssociativeBinaryOperator})
   n39 = "Sin (Radians)"(A=float * float(A=float * float(A=0.4, B=-0.8) {is=title K2Node_PromotableOperator}, B=@n19) {is=title K2Node_CommutativeAssociativeBinaryOperator})
   n40 = "Sin (Radians)"(A=float * float(A=float * float(A=1.5, B=1.2) {is=title K2Node_PromotableOperator}, B=@n19) {is=title K2Node_CommutativeAssociativeBinaryOperator})
   Set Translate Shard 06 r(
    execute=@n1.then_12
    Translate Shard 06 r:Vector=MakeVector(X=float / float(A=@n38, B=-3.5), Y=float / float(A=@n40, B=1), Z=float / float(A=@n39, B=-2))
   ) -> (Output_Get:Vector)
   n41 = "Sin (Radians)"(A=float * float(A=float * float(A=1.5, B=1.25) {is=title K2Node_PromotableOperator}, B=@n19) {is=title K2Node_CommutativeAssociativeBinaryOperator})
   n42 = "Sin (Radians)"(A=float * float(A=float * float(A=1, B=0.8) {is=title K2Node_PromotableOperator}, B=@n19) {is=title K2Node_CommutativeAssociativeBinaryOperator})
   n43 = "Sin (Radians)"(A=float * float(A=float * float(A=0.4, B=-0.8) {is=title K2Node_PromotableOperator}, B=@n19) {is=title K2Node_CommutativeAssociativeBinaryOperator})
   Set Translate Shard 07 r(
    execute=@n1.then_13
    Translate Shard 07 r:Vector=MakeVector(X=float / float(A=@n42, B=-3), Y=float / float(A=@n41, B=1), Z=float / float(A=@n43, B=-2))
   ) -> (Output_Get:Vector)
   Set Float A(
    execute=@n1.then_14
    Float A:Vector=MakeWaveVector!(Amplitudes:Rotator="1.2,0.5,0.9", Frequencies:Rotator="100,50,70", Offsets:Rotator="0,0,0", bUseNoise:bool=false, Min:Vector, Max:Vector) -> (ReturnValue:Vector)
   ) -> (Output_Get:Vector)
   > Set Float B(Float B:Vector=MakeWaveVector(Amplitudes="1.2,0.5,0.9", Frequencies="100,50,70", Offsets="-45,-45,70", bUseNoise=false)) -> (Output_Get:Vector)
   > Set Float C(Float C:Vector=MakeWaveVector(Amplitudes="1,0.5,0.9", Frequencies="100,50,70", Offsets="-35,-45,30", bUseNoise=false)) -> (Output_Get:Vector)
   > Set Float D(Float D:Vector=MakeWaveVector(Amplitudes="1.3,0.5,0.9", Frequencies="100,50,75", Offsets="-25,0,-25", bUseNoise=false)) -> (Output_Get:Vector)
   > Set Float E(Float E:Vector=MakeWaveVector(Amplitudes="1,0.5,1", Frequencies="80,50,75", Offsets="-30,0,-25", bUseNoise=false)) -> (Output_Get:Vector)
   TryGetPawnOwner()
  ]}
]
variables [
 Float A {type=Vector}
 Float B {type=Vector}
 Float C {type=Vector}
 Float D {type=Vector}
 Float E {type=Vector}
 Gravity Override Params {type=GravityOverrideParamsStruct}
 GravityOverride {type=Vector}
 Skydive Alpha {type=number}
 Translate Shard 01 l {type=Vector}
 Translate Shard 01 r {type=Vector}
 Translate Shard 02 l {type=Vector}
 Translate Shard 02 r {type=Vector}
 Translate Shard 03 l {type=Vector}
 Translate Shard 03 r {type=Vector}
 Translate Shard 04 l {type=Vector}
 Translate Shard 04 r {type=Vector}
 Translate Shard 05 l {type=Vector}
 Translate Shard 05 r {type=Vector}
 Translate Shard 06 l {type=Vector}
 Translate Shard 06 r {type=Vector}
 Translate Shard 07 r {type=Vector}
]
